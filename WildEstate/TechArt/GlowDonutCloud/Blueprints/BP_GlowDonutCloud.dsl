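; BP_GlowDonutCloud is the blueprint for a glowing donut gas cloud with a large low-gravity wind Niagara effect.
; On BeginPlay it waits one frame, then enables custom depth rendering on the gas cloud static mesh.
; After a short further delay it disables custom depth on the wind effect, a step noted as being for the VFX outline.
; The construction script is empty.
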(event EventBeginPlay
  ;[1 1 1] Custom Depth
  (Utilities|FlowControl|DelayUntilNextFrame)
  (Rendering|SetRenderCustomDepth (Variables|Default|GetGasCloud_A2_StaticMeshComponent0) true)
  ;[1 1 1] for VFX Outline
  (Utilities|FlowControl|Delay 0.3)
  (Rendering|SetRenderCustomDepth (Variables|Default|GetNS_WildEstate_LowGrav_LargeWind) false))

(fn ConstructionScript)
