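# Unlit, two-sided translucent material for a glowing, animated area wall effect, usable on Niagara sprites and meshes.
# Panning tiled textures, paced by a particle dynamic parameter, combine with noise-distorted plasma into tinted, time-of-day-scaled emissive.
# Opacity multiplies a scrolling trail mask, Fresnel on panned normals, vertex red and UV gradient with near, far and depth fades.
# A platform switch lowers base alpha on Switch, Android and PC performance mode.
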Material /WildEstate/TechArt/GlowDonutCloud/FX/M_WildEstate_AreaWall
graph {
 properties {
  bEnableResponsiveAA=true
  bUsedWithMeshParticles=true
  bUsedWithNanite=true
  bUsedWithNiagaraMeshParticles=true
  bUsedWithNiagaraSprites=true
  blendMode=BLEND_Translucent
  pixelDepthOffsetMode=PDOM_Legacy
  shadingModel=MSM_Unlit
  translucencyLightingMode=TLM_VolumetricPerVertexDirectional
  twoSided=true
  usedShadingModels=Unlit
 }
 nodes [
  Add()
  n1 = VectorParameter() {parameterName=Panner, defaultValue=<rgba 0.01 0.004 0.014 0.006>}
  AppendVector(A=@n1.R, B=@n1.R)
  n2 = VectorParameter() {parameterName=Color, defaultValue=<rgba 2.9485 0.2917 3.5 0.5>}
  n3 = Multiply(A=@n2.A) {constB=0.7}
  n4 = TextureCoordinate()
  n5 = TextureCoordinate()
  n6 = Multiply(A=ComponentMask(None=VectorParameter() {parameterName=UV, defaultValue=<rgba 20 2 50 4>}.RGBA) {a=true, b=true}, B=@n5)
  n7 = MaterialFunctionCall(XY=@n6) {materialFunction=&Swizzle}
  n8 = ScalarParameter() {defaultValue=0.22, parameterName=Ua}
  n9 = ScalarParameter() {defaultValue=0.21, parameterName=Va}
  n10 = Multiply(A=Time(), B=DynamicParameter() {paramNames [PannerSpeed,Param2,Param3,Param4]}.PannerSpeed)
  n11 = VectorParameter() {parameterName=Panner, defaultValue=<rgba 0.01 0.004 0.014 0.006>}
  n12 = Panner(
   Coordinate=Multiply(A=StaticSwitchParameter(True=@n7.YX, False=@n6) {defaultValue=true, parameterName=Swizzle}, B=AppendVector(A=@n8, B=@n9))
   Time=@n10
   Speed=AppendVector(A=@n11.B, B=@n11.B)
  ) {bFractionalPart=true, speedX=0.763, speedY=0.071}
  n13 = ScalarParameter() {defaultValue=0.15, parameterName=Ub}
  n14 = ScalarParameter() {defaultValue=0.07, parameterName=Vb}
  n15 = VectorParameter() {parameterName=Panner, defaultValue=<rgba 0.01 0.004 0.014 0.006>}
  n16 = Panner(
   Coordinate=Multiply(A=StaticSwitchParameter(True=@n7.YX, False=@n6) {defaultValue=true, parameterName=Swizzle}, B=AppendVector(A=@n13, B=@n14))
   Time=@n10
   Speed=AppendVector(A=@n15.R, B=@n15.R)
  ) {bFractionalPart=true, speedX=0.61, speedY=-0.06}
  n17 = TextureSample(UVs=@n16) {texture=&2ndInfectedMapLinear}
  n18 = Normalize(VectorInput=Add(
    A=TextureSample(UVs=@n12) {samplerType=SAMPLERTYPE_Normal, texture=/Game/Global/Textures/Default/Tileable_Patterns/2ndInfectedMapNormal}.RGB
    B=@n17.RGB
   ))
  n19 = Multiply(
   A=TextureSample(UVs=Panner(Coordinate=@n4) {bFractionalPart=true, speedX=0.05}) {texture=/Game/Effects/Fort_Effects/Textures/Tile/T_Trail_Masks_Tiled}.G
   B=MaterialFunctionCall(Exponent=ScalarParameter() {defaultValue=4, parameterName=Fresnel Exponent}, Vertex / Pixel Normal=Transform(None=@n18)) {materialFunction=&MF_Fresnel_FX}.Fresnel Double Sided
  )
  n20 = VertexColor()
  n21 = ScalarParameter() {defaultValue=500, parameterName=Near Fade}
  n22 = Multiply(
   A=Multiply(
    A=MaterialFunctionCall("Xb1,PS4"=@n2.A, "PC-High&&Epic,XSX,PS5"=@n2.A, "PC-Perf,Android"=@n3, Switch=@n3, "PC-Low(NotPerf)&&Medium"=@n2.A) {materialFunction=&PlatformSwitch_1D}.Result
    B=Multiply(A=@n19, B=@n20.R)
   )
   B=Power(Base=MaterialFunctionCall(Distance=@n21) {materialFunction=&Near_Camera_Fade}.Result) {constExponent=3.5}
  )
  n23 = Power(Base=ComponentMask(None=TextureCoordinate()) {g=true})
  n24 = MaterialFunctionCall(Distance=ScalarParameter() {defaultValue=10000, parameterName=Far Fade}) {materialFunction=&Near_Camera_Fade}
  n25 = Add(A=Power(Base=OneMinus(None=@n24.Result), Exp=ScalarParameter() {defaultValue=1.5, parameterName=Param_1})) {constB=0.3333}
  DepthFade(Opacity=Multiply(A=Multiply(A=@n22, B=@n23), B=Clamp(None=@n25)), FadeDistance=ScalarParameter() {defaultValue=250, parameterName=DepthFade})
  n26 = Add(A=Panner(Coordinate=TextureCoordinate() {uTiling=11, vTiling=2}) {bFractionalPart=true, speedX=0.05, speedY=-0.04}, B=@n5)
  n27 = Subtract(A=ComponentMask(None=TextureSample(UVs=@n26) {samplerType=SAMPLERTYPE_LinearColor, texture=/Game/Effects/Fort_Effects/Textures/Tile/T_PerlinRGB_Displacement}.RGB) {g=true, r=true}) {constB=0.5}
  n28 = ScalarParameter() {defaultValue=0.05, parameterName=Turbu Intensity}
  n29 = TextureSample(UVs=Add(A=Multiply(A=@n27, B=@n28), B=@n12)) {samplerType=SAMPLERTYPE_LinearGrayscale, texture=/Game/Effects/Fort_Effects/Emotes/UltraLight/Texture/T_Plasma_A_Greyscale}
  n30 = ScalarParameter() {defaultValue=0.8}
  n31 = Divide(A=Add(A=Power(Base=@n29.RGB, Exp=@n30) {constExponent=1.5}, B=@n17.RGB))
  n32 = VectorParameter() {parameterName=ToD, defaultValue=<rgba 0.5 0.55 0.5 0.5>}
  MaterialFunctionCall(Emissive=Multiply(
    A=Multiply(A=@n31, B=@n2.RGB)
    B=MaterialFunctionCall(Value_Morning=@n32.R, Value_Day=@n32.R, Value_Evening=@n32.R, Value_Night=@n32.R) {materialFunction=&Scalar_Time_of_Day}.Result
   )) {materialFunction=&MF_FX_EmissiveAdjustments}
  Multiply(A=@n18, B=VertexNormalWS())
  n33 = ComponentMask(None=VectorParameter() {parameterName=UV, defaultValue=<rgba 20 2 50 4>}.RGB) {g=true, r=true}
  StaticSwitchParameter(
   True=MaterialFunctionCall(XY=Multiply(A=@n33, B=@n4)) {materialFunction=&Swizzle}.YX
   False=@n4
  ) {defaultValue=true, parameterName=Swizzle}
  TextureSample(UVs=@n16) {samplerType=SAMPLERTYPE_Normal, texture=&2ndInfectedMapNormal}
 ]
}
