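# Niagara system for large, slow-orbiting mesh effects, named as low-gravity wind in the WildEstate GlowDonutCloud effects.
# Its one working emitter bursts a dozen long-lived mesh particles, partitioned four ways into a visibility tag to split them across four mesh renderers.
# Each particle slowly spins and circles an elevated center on a tilted, very wide path, with randomized radius, phase and masked height and scale.
# Two other emitters are empty.
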NiagaraSystem /WildEstate/TechArt/GlowDonutCloud/FX/NS_WildEstate_LowGrav_LargeWind
dependencies {usedDataInterfaces [], usedDynamicInputs [], usedModules [&SystemState]}
emitters {
 Empty {rendererClasses [], renderers []}
 Empty001 {rendererClasses [], renderers []}
 OmnidirectionalBurst {
  emitterUpdateScript [
   &EmitterState(Inactive Response=::NewEnumerator1, Life Cycle Mode=::NewEnumerator1, Loop Duration=10, UseLoopDelay=false)
   &SpawnBurst_Instantaneous(Spawn Count=2, Use Spawn Probability=false, UseLoopCountLimit=false) {enabled=false}
   &SpawnBurst_Instantaneous(Spawn Count=12, Spawn Group=1, Use Spawn Probability=false, UseLoopCountLimit=true) {module=SpawnBurst_Instantaneous001}
  ]
  particleSpawnScript [
   &PartitionParticles(Number Of Partitions=4)
   /SetVariables(Particles.VisibilityTag=$Particles.PartitionParticles.Partition) {isSetParameters=true}
   &InitializeParticle(
    Mesh Scale Mode=::NewEnumerator3
    Mesh Uniform Scale=&MaskFloat(
     Mask Source=0
     Mask Parameter=?
     Mask=0
     Masked Float=1
     Passthrough Float=&RandomRangeFloat(Minimum=1, Maximum=1.5)
    )
    Sprite Rotation Mode=::NewEnumerator2
    Sprite Size Mode=::NewEnumerator3
    Uniform Sprite Size=1000
    UsePositionOffset=false
    Write Material Random=true
   )
   &InitialMeshOrientation(Use Rotation Vector=false)
   &DynamicMaterialParameters(Module.Index0Float0Write=true, Module.Index0Float1Write=true, Module.Index0Float2Write=true, Module.Index0Float3Write=true) {enabled=false}
  ]
  particleUpdateScript [
   &ParticleState(Kill Particles When Lifetime Has Elapsed=false)
   &UpdateMeshOrientation(Rotation Rate=0.05)
   &RotateAroundPoint(
    Radius:NiagaraFloat=&RandomRangeFloat(Minimum=10500, Maximum=12500)
    Rotation Center:NiagaraPosition=&MaskVector(
     Mask Source=0
     Mask Parameter=?
     Mask=0
     Masked Vector=<0 0 10500>
     Passthrough Vector=&RandomRangeVector(Minimum=<0 0 6500>, Maximum=<0 0 8500>)
    )
    Rotation Phase:NiagaraFloat=&RandomRangeFloat(Minimum=0, Maximum=1)
    Rotation Rate=0.004
    X Axis:Vector3f=<1 0 0.3>
   )
  ]
  rendererClasses [&NiagaraMeshRendererProperties]
  renderers [
   {rendererClass=&NiagaraMeshRendererProperties, rendererIndex=0}
   {rendererClass=&NiagaraMeshRendererProperties, rendererIndex=1}
   {rendererClass=&NiagaraMeshRendererProperties, rendererIndex=2}
   {rendererClass=&NiagaraMeshRendererProperties, rendererIndex=3}
  ]
 }
}
summary {
 systemName=NS_WildEstate_LowGrav_LargeWind
 emitters [
  {emitterName=Empty, rendererClasses []}
  {emitterName=OmnidirectionalBurst, rendererClasses [&NiagaraMeshRendererProperties]}
  {emitterName=Empty001, rendererClasses []}
 ]
 userVariables []
}
