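# Masked unlit material for a stylized glowing smoke cloud, with an IsBody switch separating the inset body from the sprite puff shell.
# WPO expands vertex-colored sprites scaled by thickness, view angle and distance, adding sine wobble, rotating puff textures and pixel depth offset.
# Color combines stepped sun and height shading, time-of-day tints, outlines and an animated triplanar-noise interior glow.
# Screen-aligned noise drives a height-lerped near-camera fade with intersection outlines.
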Material /WildEstate/TechArt/GlowDonutCloud/Materials/M_SmokeCloud_PPWeek3
graph {
 properties {
  bAutomaticallySetUsageInEditor=false
  bHasPixelAnimation=true
  bUsedWithMeshDeformer=true
  bUsedWithSkeletalMesh=true
  blendMode=BLEND_Masked
  shadingModel=MSM_Unlit
  usedShadingModels=Unlit
 }
 nodes [
  n1 = WorldPosition()
  n2 = MaterialFunctionCall(Color=@n1.XYZ) {materialFunction=&SplitComponents}
  n3 = ComponentMask(None=@n1.XYZ) {b=true}
  n4 = VectorParameter() {group=InteriorGLow, parameterName=SphereLocation}
  n5 = MaterialFunctionCall(R=@n2.R, G=@n2.R, B=If(A=@n3, B=@n4.B, "A > B"=@n3, "A < B"=@n4.B)) {materialFunction=&AppendMany}
  n6 = ScalarParameter() {defaultValue=1.2486, group=InteriorGLow, parameterName=InteriorGlow_Radius}
  n7 = ScalarParameter() {defaultValue=0.1, group=InteriorGLow, parameterName=InteriorGlow_Hardness}
  Abs(None=OneMinus(None=SphereMask(A=@n5.RGB, B=@n4.RGB, Radius=@n6, Hardness=@n7)))
  n8 = ScalarParameter() {defaultValue=0.8, parameterName=AngleMask.Min}
  n9 = ScalarParameter() {defaultValue=1, parameterName=AngleMask.Max}
  n10 = CameraPositionWS()
  n11 = Add(A=ObjectPositionWS(), B=Constant3Vector() {constant=<rgba 0 0 1000 0>})
  n12 = DotProduct(A=Normalize(VectorInput=Subtract(A=@n10, B=@n11)), B=Constant3Vector() {constant=<rgba 0 0 -1 0>})
  n13 = LinearInterpolate(Alpha=SmoothStep(Min=@n8, Max=@n9, Value=ConstantBiasScale(None=@n12)) {constMin=0.97}) {constA=2, constB=0}
  n14 = ComponentMask(None=TextureCoordinate()) {g=true}
  Add(A=@n13, B=Subtract(A=OneMinus(None=@n14)))
  n15 = ActorPositionWS()
  n16 = MaterialFunctionCall(Vector 2=ComponentMask(None=@n15) {b=true, r=true}) {materialFunction=/Engine/Functions/Engine_MaterialFunctions02/Utility/VectorLength}
  n17 = ComponentMask(None=@n15) {g=true}
  n18 = Subtract(A=Abs(None=AppendVector(A=@n16.V2 Length, B=@n17)), B=ScalarParameter() {defaultValue=512, group=Cylindarmask, parameterName=CylMaskHeight})
  n19 = ComponentMask(None=@n18) {r=true}
  n20 = ComponentMask(None=@n18) {g=true}
  Add(A=Distance(A=Max(A=@n18) {constB=0}, B=Constant()), B=Min(A=Max(A=@n19, B=@n20) {constB=0}) {constB=0})
  n21 = VertexNormalWS()
  n22 = Multiply(
   A=ScalarParameter() {defaultValue=-1, group=WPO, parameterName=BodyInset}
   B=MaterialFunctionCall() {materialFunction=&ObjectScale}.Scale Z
  )
  n23 = LinearInterpolate(
   A=ScalarParameter() {defaultValue=6.202, group=WPO, parameterName=Sprite_Scale_Thin}
   B=ScalarParameter() {defaultValue=14.0368, group=WPO, parameterName=Sprite_Scale_Sides}
   Alpha=NamedRerouteUsage()
  )
  n24 = Multiply(A=ScalarParameter() {defaultValue=3.9214, group=WPO, parameterName=Sprite_Scale_Random}, B=Subtract(A=VertexColor().R) {constB=0.5})
  n25 = ScalarParameter() {defaultValue=-0.9246, group=WPO, parameterName=Sprite_Scale_Back_Thin}
  n26 = ScalarParameter() {defaultValue=29.4715, group=WPO, parameterName=Sprite_Scale_Back}
  n27 = NamedRerouteUsage()
  n28 = WorldPosition() {worldPositionShaderOffset=WPT_CameraRelativeNoOffsets}
  n29 = Normalize(VectorInput=@n28.XYZ)
  n30 = NamedRerouteDeclaration(None=DotProduct(A=VertexNormalWS(), B=NamedRerouteDeclaration(None=@n29) {name=CameraVector_Custom, nodeColor=<rgba 0 1 0.8828 1>})) {name=DotSurface, nodeColor=<rgba 0.2422 0 1 1>}
  n31 = MaterialFunctionCall() {materialFunction=&ObjectScale}
  n32 = Subtract(
   A=Length(None=@n28.XYZ) {is=class MaterialExpressionLength}
   B=Multiply(A=ScalarParameter() {group=WPO, parameterName=Distance_Offset}, B=@n31.Scale Z)
  )
  n33 = Multiply(A=ScalarParameter() {defaultValue=29.5774, group=WPO, parameterName=Distance_Blend}, B=@n31.Scale Z)
  n34 = LinearInterpolate(
   A=ScalarParameter() {defaultValue=4, group=WPO, parameterName=ScaleNear}
   B=LinearInterpolate(A=Add(A=@n23, B=@n24), B=LinearInterpolate(A=@n25, B=@n26, Alpha=@n27), Alpha=Saturate(None=@n30))
   Alpha=Saturate(None=Divide(A=@n32, B=@n33))
  )
  n35 = MaterialFunctionCall() {materialFunction=&ObjectScale}
  n36 = ScalarParameter() {defaultValue=-2.2442, group=WPO, parameterName=SurfaceOffset_thin}
  n37 = ScalarParameter() {defaultValue=-4.9573, group=WPO, parameterName=SurfaceOffset}
  n38 = NamedRerouteUsage()
  n39 = Saturate(None=NamedRerouteUsage())
  n40 = Multiply(A=VertexNormalWS(), B=Multiply(A=LinearInterpolate(A=@n36, B=@n37, Alpha=@n38), B=OneMinus(None=@n39)))
  n41 = Multiply(A=NamedRerouteUsage()) {constB=-1}
  n42 = ScalarParameter() {defaultValue=-4.8221, group=WPO, parameterName=PushBackFront_thin}
  n43 = ScalarParameter() {defaultValue=-4.3507, group=WPO, parameterName=PushBackFront}
  n44 = NamedRerouteUsage()
  n45 = Multiply(A=NamedRerouteUsage(), B=Multiply(A=Saturate(None=@n41), B=LinearInterpolate(A=@n42, B=@n43, Alpha=@n44)))
  n46 = Add(
   A=MaterialFunctionCall!(Center Location=WorldPosition().XYZ, 0-1 UVs, XY scale=Multiply(A=@n34, B=@n35.Scale Z), OverridePivot, "Normals (Optional)") {materialFunction=&Sprite}.World position offset
   B=Multiply(
    A=Add(A=@n40, B=@n45)
    B=MaterialFunctionCall() {materialFunction=&ObjectScale}.Scale Z
   )
  )
  n47 = ScalarParameter() {defaultValue=1, group=WPO, parameterName=OverallScale}
  Add(
   A=StaticSwitchParameter(True=Multiply(A=@n21, B=@n22), False=Multiply(A=@n46, B=@n47)) {group=00_Switches, parameterName=IsBody, sortPriority=0}
   B=NamedRerouteUsage()
  )
  n48 = ScalarParameter() {defaultValue=3, group=Color, parameterName=GlowContrast}
  n49 = NamedRerouteUsage()
  n50 = Multiply(
   A=VectorParameter() {group=Color, parameterName=OutlineGlowColor, defaultValue=<rgba 0.9592 0 1.33 1>}.RGB
   B=MaterialFunctionCall(Value_Morning=Constant() {r=1}, Value_Day=Constant() {r=1.075}, Value_Evening=Constant() {r=0.95}, Value_Night=Constant() {r=0.875}) {materialFunction=&Scalar_Time_of_Day}.Result
  )
  n51 = Multiply(A=LinearInterpolate(A=ScalarParameter() {defaultValue=1.15, group=Color, parameterName=BrightnessFace}, Alpha=VertexColor().B), B=@n49)
  n52 = LinearInterpolate(A=@n51, B=VectorParameter() {group=Color, parameterName=OutlineColor, defaultValue=<rgba 0.05 0.0345 0.125 1>}.RGB)
  n53 = Saturate(None=LinearInterpolate(Alpha=NamedRerouteUsage()) {constA=-1})
  n54 = Saturate(None=Add(A=NamedRerouteUsage(), B=@n53))
  n55 = StaticSwitchParameter(True=@n49, False=LinearInterpolate(A=LinearInterpolate(A=@n50, B=@n52, Alpha=@n54), B=@n51, Alpha=@n53)) {group=00_Switches, parameterName=IsBody, sortPriority=0}
  n56 = Add(A=Multiply(A=Power(Base=@n55, Exp=@n48), B=ScalarParameter() {defaultValue=0.1, group=Color, parameterName=Glow}), B=@n55)
  EyeAdaptationInverse(LightValueInput=@n56)
  n57 = VertexNormalWS()
  n58 = Multiply(
   A=MaterialFunctionCall() {materialFunction=&ObjectScale}.Scale Z
   B=ScalarParameter() {defaultValue=16.1339, group=Normal, parameterName=OffsetNormal}
  )
  n59 = Multiply(A=VertexColor() {desc="Red is Random\r\nGreen Is Thickness\r\nBlue is Face Mask"}.G, B=@n57)
  n60 = Subtract(A=WorldPosition().XYZ, B=Subtract(A=WorldPosition() {worldPositionShaderOffset=WPT_ExcludeAllShaderOffsets}.XYZ, B=Multiply(A=@n58, B=@n59)))
  LinearInterpolate(A=@n57, B=@n60, Alpha=VertexColor() {desc="Red is Random\r\nGreen Is Thickness\r\nBlue is Face Mask"}.B)
  n61 = TextureObject() {samplerType=SAMPLERTYPE_LinearColor, texture=&HighContrastNoise}
  n62 = VectorParameter() {desc=37.719421, group=InteriorGLow, parameterName=CausticWorldScale, defaultValue=<rgba 8500 8500 8500 0>}
  n63 = WorldPosition() {worldPositionShaderOffset=WPT_ExcludeAllShaderOffsets}
  n64 = Multiply(A=Time()) {constB=0.05}
  n65 = StaticBool()
  n66 = Constant() {r=1}
  n67 = MaterialFunctionCall!(TextureObject=@n61, TextureSize=@n62.RGB, WorldPosition, Export Float 4=@n65, In, World Space Normal, ProjectionTransitionContrast=@n66) {materialFunction=/Game/Athena/SafeZone/WorldAlignedTexture_ScalePan}
  LinearInterpolate(
   A=MaterialFunctionCall(
    TextureObject=@n61
    TextureSize=Multiply(A=@n62.RGB) {constB=0.88}
    WorldPosition=Add(A=@n63.XYZ, B=@n64)
    Export Float 4=@n65
    ProjectionTransitionContrast=@n66
   ) {materialFunction=&WorldAlignedTexture_ScalePan}.XYZ Texture
   B=@n67.XYZ Texture
  )
  MaterialFunctionCall(Emissive=@n56) {materialFunction=&MF_FX_EmissiveAdjustments}
  n68 = ScalarParameter() {defaultValue=0.1, group=InteriorGLow, parameterName=InteriorGlow_SineSpeed}
  MaterialFunctionCall(Speed=@n68) {materialFunction=&DebugTimeSine}
  n69 = Fresnel(Normal=NamedRerouteUsage()) {baseReflectFraction=0, exponent=8}
  n70 = ScalarParameter() {defaultValue=1, group=Color, parameterName=InnerGlowShadinSteps}
  n71 = Divide(A=Round(None=Multiply(A=@n69, B=@n70)), B=@n70)
  Multiply(A=@n71, B=NamedRerouteUsage())
  Multiply(
   A=NamedRerouteUsage()
   B=StaticSwitchParameter(True=Constant() {r=1}, False=NamedRerouteUsage()) {group=00_Switches, parameterName=IsBody, sortPriority=0}
  )
  n72 = Constant3Vector() {constant=<rgba 1 0 0 0>}
  n73 = MaterialFunctionCall(
   Plane Normal Axis=CameraVectorWS()
   Plane X Axis=Normalize(VectorInput=Transform(None=@n72) {transformSourceType=TRANSFORMSOURCE_Camera})
   Plane Center=ObjectPositionWS()
   UV Scale=Multiply(
    A=ScalarParameter() {defaultValue=500, group=Noise, parameterName=Noise_Size, sortPriority=36}
    B=MaterialFunctionCall() {materialFunction=&ObjectScale}.Scale Z
   )
  ) {materialFunction=&VirtualPlaneCoordinates}
  NamedRerouteDeclaration(None=TextureSample(UVs=@n73.UVs) {samplerType=SAMPLERTYPE_Masks, texture=/WildEstate/Assets/FX/GasCloud/GasCloud_Tiling_M}.RGB) {name=NoiseTexture, nodeColor=<rgba 0.4297 1 0 1>}
  n74 = Distance(A=WorldPosition().XYZ, B=CameraPositionWS())
  n75 = ScalarParameter() {defaultValue=1594.8198, group=CameraFade, parameterName=CameraFade_Offset}
  n76 = Saturate(None=Divide(A=Subtract(A=@n74, B=@n75), B=ScalarParameter() {defaultValue=953.3473, group=CameraFade, parameterName=CameraFade_Distance}))
  n77 = MaterialFunctionCall(Transition Phase=@n76, Height Texture=ComponentMask(None=NamedRerouteUsage()) {r=true}) {materialFunction=&HeightLerp}
  n78 = Subtract(A=@n77.Alpha, B=ScalarParameter() {defaultValue=-0.6584, group=CameraFade, parameterName=IntersectionOutline})
  n79 = ComponentMask(None=NamedRerouteUsage()) {g=true}
  n80 = Floor(None=@n76)
  NamedRerouteDeclaration(None=Multiply(A=Floor(None=@n78), B=LinearInterpolate(A=@n79, Alpha=@n80) {constA=1})) {name=OutlinesIntersection, nodeColor=<rgba 0.0547 0 1 1>}
  n81 = Saturate(None=@n13)
  NamedRerouteDeclaration(None=@n81) {name=CameraAngleMask_Bottom, nodeColor=<rgba 0 1 0.2969 1>}
  NamedRerouteDeclaration(None=TextureSample(UVs=@n73.UVs) {samplerType=SAMPLERTYPE_Grayscale, texture=&LowResBlurredNoise}.R) {name=NoiseTexture_Gray, nodeColor=<rgba 0.4297 1 0 1>}
  n82 = ComponentMask(None=NamedRerouteUsage()) {b=true}
  n83 = ScalarParameter() {defaultValue=0.01, group=InteriorGLow, parameterName=InteriorGlow_NoiseAmount}
  n84 = Subtract(A=WorldPosition() {worldPositionShaderOffset=WPT_ExcludeAllShaderOffsets}.XYZ, B=ObjectPositionWS())
  n85 = Add(A=Multiply(A=@n82, B=@n83), B=ComponentMask(None=@n84) {b=true}) {constB=-500}
  NamedRerouteDeclaration(None=Saturate(None=OneMinus(None=@n85))) {name=ZUPMask, nodeColor=<rgba 1 0.9141 0 1>}
  n86 = ComponentMask(None=Add(A=@n84, B=Multiply(A=@n83, B=@n82) {constA=0.25})) {g=true, r=true}
  n87 = Constant2Vector() {g=5, r=0.5}
  n88 = ScalarParameter() {defaultValue=5.01, group=InteriorGLow, parameterName=InteriorGlow_SphereRadiusMax}
  n89 = LinearInterpolate(A=ScalarParameter() {defaultValue=1.2486, group=InteriorGLow, parameterName=InteriorGlow_SphereRadius}, B=@n88, Alpha=NamedRerouteUsage())
  n90 = ScalarParameter() {defaultValue=0.1, group=InteriorGLow, parameterName=InteriorGlow_SphereHardness}
  NamedRerouteDeclaration(None=Saturate(None=SphereMask(A=@n86, B=@n87, Radius=@n89, Hardness=@n90))) {name=BoundMask, nodeColor=<rgba 0.3594 0 1 1>}
  n91 = Multiply(A=Time() {desc=15.015014, period=1}, B=ScalarParameter() {defaultValue=0.08, desc=0.0666, group=InteriorGLow, parameterName=Sine Speed A_1})
  n92 = ComponentMask(None=Frac(None=Add(A=@n67.XYZ Texture, B=@n91))) {g=true}
  n93 = Constant() {r=0.5}
  n94 = OneMinus(None=Multiply(A=Distance(A=@n92, B=@n93)) {constB=2})
  NamedRerouteDeclaration(None=LinearInterpolate(A=@n94, B=Round(None=@n94), Alpha=ScalarParameter() {defaultValue=0.034, parameterName=Caustic Round LERP})) {name=CenterNoise, nodeColor=<rgba 0.8516 0 1 1>}
  n95 = Multiply(A=Time(), B=@n68)
  n96 = Multiply(A=@n95) {constB=4.5}
  n97 = Multiply(A=Add(A=Cosine(None=@n96), B=@n95)) {constB=2}
  n98 = Multiply(A=@n95) {constB=3}
  n99 = Add(A=Multiply(A=Sine(None=@n97)) {constB=0.66}, B=Multiply(A=Sine(None=@n98)) {constB=0.5})
  NamedRerouteDeclaration(None=Saturate(None=ConstantBiasScale(None=@n99))) {name=Jitter, nodeColor=<rgba 0.7344 1 0 1>}
  n100 = VectorParameter() {group=Color, parameterName=Glow, defaultValue=<rgba 0.7136 0.0141 0.7969 1>}
  n101 = MaterialFunctionCall(Value_Morning=Constant() {r=1}, Value_Day=Constant() {r=1.05}, Value_Evening=Constant() {r=0.9}, Value_Night=Constant() {r=0.6}) {materialFunction=&Scalar_Time_of_Day}
  n102 = NamedRerouteUsage()
  n103 = Max(A=Multiply(A=@n71, B=NamedRerouteUsage()), B=@n102)
  NamedRerouteDeclaration(None=LinearInterpolate(A=NamedRerouteUsage(), B=Multiply(A=@n100.RGB, B=@n101.Result), Alpha=@n103)) {name=InnerGLow, nodeColor=<rgba 0.4531 0 1 1>}
  NamedRerouteDeclaration(None=SphereMask(A=@n86, B=Constant2Vector() {g=5, r=0.5}, Radius=@n88) {hardnessPercent=0.5}) {name=SphereMask, nodeColor=<rgba 0 0.2969 1 1>}
  n104 = OneMinus(None=TextureSample() {samplerType=SAMPLERTYPE_LinearGrayscale, texture=/WildEstate/Assets/FX/GasCloud/GasCloud_height}.R)
  n105 = Multiply(
   A=ScalarParameter() {defaultValue=3, group=Pixel Depth Offset, parameterName=PDO}
   B=MaterialFunctionCall() {materialFunction=&ObjectScale}.Scale Z
  )
  NamedRerouteDeclaration(None=StaticSwitchParameter(True=Constant(), False=Multiply(A=@n104, B=@n105)) {group=00_Switches, parameterName=IsBody, sortPriority=0}) {name=PDO, nodeColor=<rgba 1 0 0.2812 1>}
  NamedRerouteDeclaration(None=@n103) {name=InnerGlowMask, nodeColor=<rgba 0 0.2031 1 1>}
  n106 = VectorParameter() {group=TODM, parameterName=ColorTint_Morning, defaultValue=<rgba 1 0.6979 0.7786 1>}
  n107 = VectorParameter() {group=TODM, parameterName=ColorTint_Day, defaultValue=<rgba 0.8333 0.8333 0.8333 1>}
  n108 = VectorParameter() {group=TODM, parameterName=ColorTint_Evening, defaultValue=<rgba 0.7326 0.483 0.8667 1>}
  n109 = VectorParameter() {group=TODM, parameterName=ColorTint_Night, defaultValue=<rgba 0.415 0.63 0.7812 1>}
  n110 = ScalarParameter() {defaultValue=0.5849, group=Color, parameterName=Brightness}
  n111 = VectorParameter() {group=Color, parameterName=Color2, defaultValue=<rgba 0.1133 0.0023 0.3564 1>}
  n112 = VectorParameter() {group=Color, parameterName=Color, defaultValue=<rgba 0.1777 0.0175 0.491 1>}
  n113 = NamedRerouteUsage()
  n114 = CollectionParameter() {
   collection=&FortniteMaterialParameters
   parameterName=SunlightDirectionalVector
  }
  n115 = NamedRerouteUsage()
  n116 = ComponentMask(None=VertexNormalWS()) {b=true}
  n117 = Add(B=@n116) {constA=-0.1}
  n118 = Max(A=DotProduct(A=NamedRerouteUsage(), B=@n114), B=Add(A=Multiply(A=@n115) {constB=0.02}, B=@n117))
  n119 = StaticSwitchParameter(True=@n118, False=Max(A=@n118, B=@n117)) {group=00_Switches, parameterName=IsBody, sortPriority=0}
  n120 = Divide(A=Add(A=Add(A=@n113, B=@n119), B=ScalarParameter() {defaultValue=1.2641, group=Color, parameterName=ShadinOffset}))
  n121 = Add(A=Multiply(A=@n115) {constB=0.3}, B=Add(A=@n116) {constB=0.7})
  n122 = ComponentMask(None=WorldPosition().XYZ) {b=true}
  n123 = ComponentMask(None=MaterialFunctionCall() {materialFunction=&ObjectPivotPoint}.Object Pivot Location) {b=true}
  n124 = MaterialFunctionCall() {materialFunction=&ObjectLocalBounds}
  n125 = Saturate(None=Divide(A=Subtract(A=@n122, B=@n123), B=ComponentMask(None=@n124.Local Bounds Size) {b=true}))
  n126 = Power(Base=Add(
    A=StaticSwitchParameter(True=@n118, False=@n125) {group=00_Switches, parameterName=IsBody, sortPriority=0}
    B=ScalarParameter() {defaultValue=1.75, group=Color, parameterName=ShadingOffsetVertexNormal}
   ) {constB=-0.5}) {constExponent=8}
  n127 = Saturate(None=Multiply(A=NamedRerouteUsage(), B=ScalarParameter() {defaultValue=2, group=Color, parameterName=ShadingOffsetNoiseAmount}) {constB=2})
  n128 = StaticSwitchParameter(True=Min(A=@n126, B=Add(A=@n126, B=@n127)), False=@n126) {group=00_Switches, parameterName=IsBody, sortPriority=0}
  n129 = ScalarParameter() {defaultValue=1, group=Color, parameterName=ShadinSteps}
  n130 = Multiply(A=Min(A=Saturate(None=@n120), B=Min(A=@n121, B=@n128)), B=@n129)
  n131 = Max(A=Divide(A=Round(None=@n130), B=@n129), B=ScalarParameter() {defaultValue=0.3016, group=Color, parameterName=MinDarkness})
  n132 = LinearInterpolate(
   A=LinearInterpolate(A=@n111.RGB, B=LinearInterpolate(A=@n111.RGB, B=@n112.RGB, Alpha=@n131), Alpha=NamedRerouteUsage())
   B=@n112.RGB
   Alpha=NamedRerouteUsage()
  )
  NamedRerouteDeclaration(None=Multiply(
    A=MaterialFunctionCall(Color_Morning=@n106.RGB, Color_Day=@n107.RGB, Color_Evening=@n108.RGB, Color_Night=@n109.RGB) {materialFunction=&Color_Time_of_Day}.Result
    B=Multiply(A=@n110, B=@n132)
   )) {name=BaseCloud, nodeColor=<rgba 0.6641 1 0 1>}
  n133 = ScalarParameter() {defaultValue=0.1, group=InteriorGLow, parameterName=InteriorGlowOutline_SphereRadius}
  NamedRerouteDeclaration(None=Saturate(None=SphereMask(A=@n86, B=@n87, Radius=@n133, Hardness=@n90))) {name=InnerglowoutlineMask, nodeColor=<rgba 0 0.5547 1 1>}
  n134 = OneMinus(None=@n102)
  NamedRerouteDeclaration(None=Power(Base=Saturate(None=@n134))) {name=InnerGlowOutlineMask_Final, nodeColor=<rgba 0.875 1 0 1>}
  NamedRerouteDeclaration(None=@n67.XYZ Texture) {name=DefaultTriPlanarNoise, nodeColor=<rgba 1 0 0.0703 1>}
  n135 = Normalize(VectorInput=Subtract(A=CameraPositionWS(), B=ObjectPositionWS()))
  n136 = Constant3Vector() {constant=<rgba 0 0 1 0>}
  n137 = SmoothStep(Min=@n8, Max=@n9, Value=ConstantBiasScale(None=DotProduct(A=@n135, B=@n136))) {constMin=0.8}
  NamedRerouteDeclaration(None=Saturate(None=LinearInterpolate(Alpha=@n137) {constB=2})) {name=CameraAngleMask_Top, nodeColor=<rgba 0 1 0.2969 1>}
  NamedRerouteDeclaration(None=Power(Base=VertexColor().G, Exp=ScalarParameter() {defaultValue=0.6469, group=WPO, parameterName=Thickness_Falloff})) {name=ThicknessMask, nodeColor=<rgba 1 0.2812 0 1>}
  NamedRerouteDeclaration(None=StaticSwitchParameter(True=@n57, False=Normalize(VectorInput=@n60)) {group=00_Switches, parameterName=IsBody, sortPriority=0}) {name=Normals, nodeColor=<rgba 1 0.7031 0 1>}
  n138 = VertexColor()
  n139 = Subtract(A=@n138.R) {constB=0.5}
  n140 = Multiply(A=Time(), B=Multiply(A=ScalarParameter() {defaultValue=1, group=Animation, parameterName=Rotation_Speed}, B=Sign(None=@n139)))
  n141 = Multiply(A=@n138.R, B=ScalarParameter() {defaultValue=44.0961, group=Animation, parameterName=RandomUVRot})
  n142 = TextureSample(UVs=Rotator(Time=Add(A=@n140, B=@n141))) {samplerType=SAMPLERTYPE_Masks, texture=/WildEstate/Assets/FX/GasCloud/GasCloud_M}
  NamedRerouteDeclaration(None=@n142.G) {name=TextureOutline, nodeColor=<rgba 0.3828 0 1 1>}
  NamedRerouteDeclaration(None=@n142.R) {name=TextureMask, nodeColor=<rgba 0 1 0.0625 1>}
  n143 = ScalarParameter() {defaultValue=0.3, group=Animation, parameterName=SineMovement_Amp}
  n144 = MaterialFunctionCall() {materialFunction=&ObjectScale}
  n145 = ScalarParameter() {defaultValue=11.1265, group=Animation, parameterName=SinSIze}
  n146 = Add(
   A=Divide(A=WorldPosition() {worldPositionShaderOffset=WPT_ExcludeAllShaderOffsets}.XYZ, B=Multiply(A=@n145, B=@n144.Scale Z))
   B=Multiply(A=ScalarParameter() {defaultValue=0.5, group=Animation, parameterName=SineMovement_Speed}, B=Time())
  )
  NamedRerouteDeclaration(None=Multiply(A=Multiply(A=@n143, B=@n144.Scale Z), B=Sine(None=@n146))) {name=SineWave Wobble, nodeColor=<rgba 0.5 1 0 1>}
  NamedRerouteDeclaration(None=@n77.Alpha) {name=NearCamera Mask, nodeColor=<rgba 0.0547 0 1 1>}
  n147 = NamedRerouteUsage()
  NamedRerouteDeclaration(None=Multiply(A=ScalarParameter() {defaultValue=0.156, group=Noise, parameterName=Noise_Add}, B=ComponentMask(None=@n147) {g=true, r=true})) {name=Noise, nodeColor=<rgba 0.2656 1 0 1>}
  NamedRerouteUsage()
  OneMinus(None=@n81)
  Power(Base=@n102)
  Saturate(None=@n117)
  TextureObject() {samplerType=SAMPLERTYPE_Normal, texture=/Game/TheOutpost/Textures/TilingWeakPointNormal}
  WorldPosition()
 ]
}
